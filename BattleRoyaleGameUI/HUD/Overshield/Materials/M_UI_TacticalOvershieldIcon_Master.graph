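# UI master material for the tactical overshield HUD icon, drawing either an intact shield with bolt or a broken shield via the IsBroken static switch.
# One packed texture supplies the masks: red for intact coverage, green to separate bolt from shield, blue for the broken shape.
# Animate scalars blend start and stop colors and alphas.
# The broken variant can mix in an adaptive dim color and uses a low-end fallback for stop alpha.
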Material /BattleRoyaleGameUI/HUD/Overshield/Materials/M_UI_TacticalOvershieldIcon_Master
graph {
 properties {blendMode=BLEND_Translucent, materialDomain=MD_UI, pixelDepthOffsetMode=PDOM_Legacy, usedShadingModels=Default Lit}
 nodes [
  n1 = TextureCoordinate()
  ComponentMask(None=@n1) {g=true}
  n2 = TextureSample(UVs=@n1) {samplerType=SAMPLERTYPE_LinearColor, texture=/BattleRoyaleGameUI/HUD/Overshield/Textures/T_UI_TacOvershield_64x}
  n3 = ScalarParameter() {defaultValue=1, group=Alpha, parameterName=BrokenShield_Alpha_Start}
  n4 = MaterialFunctionCall!(Default=ScalarParameter() {defaultValue=1, group=Alpha, parameterName=BrokenShield_Alpha_Stop}, LowEnd=Constant() {r=0.3333}) {materialFunction=/BattleRoyaleGameUI/Shared/MF_UI_LowEnd_ValueSwitcher}
  n5 = ScalarParameter() {parameterName=AnimateBrokenAlpha}
  n6 = Saturate(None=Multiply(A=@n2.B, B=LinearInterpolate(A=@n3, B=@n4.Result, Alpha=@n5)))
  n7 = ScalarParameter() {defaultValue=1, group=Alpha, parameterName=Shield_Alpha}
  n8 = LinearInterpolate(
   A=ScalarParameter() {defaultValue=1, group=Alpha, parameterName=Bolt_Alpha}
   B=ScalarParameter() {defaultValue=1, group=Alpha, parameterName=Bolt_Alpha_Stop}
   Alpha=ScalarParameter() {parameterName=AnimateBoltAlpha}
  )
  n9 = Saturate(None=DotProduct(A=@n2.R, B=LinearInterpolate(A=@n7, B=@n8, Alpha=@n2.G)))
  Saturate(None=Multiply(
    A=StaticSwitchParameter(True=@n6, False=@n9) {parameterName=IsBroken}
    B=ScalarParameter() {defaultValue=1, group=Alpha, parameterName=Final_Alpha}
   ))
  n10 = VectorParameter() {group=Color, parameterName=BrokenShield_Color_Start, defaultValue=<rgba 0 0 1 1>}
  n11 = MaterialFunctionCall!(BufferSource) {materialFunction=/BattleRoyaleGameUI/Shared/MF_UI_AdaptiveColors}
  n12 = MaterialFunctionCall(Default=@n11.Dim, LowEnd=@n11.Dim) {materialFunction=&MF_UI_LowEnd_ValueSwitcher}
  n13 = ScalarParameter() {parameterName=AdaptiveColor_BackingMix}
  n14 = VectorParameter() {group=Color, parameterName=Shield_Color, defaultValue=<rgba 0 1 0 1>}
  n15 = VectorParameter() {group=Color, parameterName=Shield_Color_Stop, defaultValue=<rgba 0 1 0 1>}
  n16 = ScalarParameter() {parameterName=AnimateShieldColor}
  n17 = VectorParameter() {group=Color, parameterName=Bolt_Color, defaultValue=<rgba 1 0 0 1>}
  n18 = VectorParameter() {group=Color, parameterName=Bolt_Color_Stop, defaultValue=<rgba 1 0 0 1>}
  n19 = ScalarParameter() {parameterName=AnimateBoltColor}
  StaticSwitchParameter(
   True=LinearInterpolate(
    A=LinearInterpolate(A=@n10.RGB, B=@n12.Result, Alpha=@n13)
    B=VectorParameter() {group=Color, parameterName=BrokenShield_Color_Stop, defaultValue=<rgba 0 0 1 1>}.RGB
    Alpha=ScalarParameter() {parameterName=AnimateBrokenColor}
   )
   False=LinearInterpolate(A=LinearInterpolate(A=@n14.RGB, B=@n15.RGB, Alpha=@n16), B=LinearInterpolate(A=@n17.RGB, B=@n18.RGB, Alpha=@n19), Alpha=@n2.G)
  ) {parameterName=IsBroken}
 ]
}
