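# Translucent UI material for the tactical overshield HUD icon, acting as a progress bar that fills from the bottom up.
# Progress_Amount offsets vertical UVs within compensated limits, and EdgeHardness sharpens the resulting mask that blends backing and fill color and opacity.
# The texture's red channel shapes the icon; green masks an optional bolt with separate alpha and optional color.
# Animation scalars blend toward stop values, with adaptive backing color, optional collection-driven fill and low-end alpha fallback.
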Material /BattleRoyaleGameUI/HUD/Overshield/Materials/M_UI_TacticalOvershield_Icon_ProgressBar
graph {
 properties {blendMode=BLEND_Translucent, materialDomain=MD_UI, pixelDepthOffsetMode=PDOM_Legacy, usedShadingModels=Default Lit}
 nodes [
  n1 = ScalarParameter() {defaultValue=0.75, group=Alpha, parameterName=BackingAlpha}
  n2 = Constant() {r=0.464}
  n3 = Add(
   A=Add(A=TextureCoordinate()) {constB=-1}
   B=LinearInterpolate(
    A=ScalarParameter() {defaultValue=0.05, parameterName=ProgressCompensation_Min}
    B=ScalarParameter() {defaultValue=0.95, parameterName=ProgressCompensation_Max}
    Alpha=ScalarParameter() {defaultValue=0.5, parameterName=Progress_Amount}
   )
  )
  n4 = Saturate(None=Multiply(A=ComponentMask(None=@n3) {g=true}, B=ScalarParameter() {defaultValue=128, parameterName=EdgeHardness}))
  n5 = LinearInterpolate(
   A=LinearInterpolate(
    A=MaterialFunctionCall!(Default=@n1, LowEnd=@n2) {materialFunction=/BattleRoyaleGameUI/Shared/MF_UI_LowEnd_ValueSwitcher}.Result
    B=ScalarParameter() {defaultValue=0.5, group=Alpha, parameterName=BackingAlphaStop}
    Alpha=ScalarParameter() {parameterName=AnimateProgressBackingAlpha}
   )
   B=ScalarParameter() {defaultValue=1, group=Alpha, parameterName=FillAlpha}
   Alpha=@n4
  )
  n6 = TextureSample() {samplerType=SAMPLERTYPE_LinearColor, texture=/BattleRoyaleGameUI/HUD/Overshield/Textures/T_UI_TacOvershield_64x}
  n7 = OneMinus(None=@n6.G)
  n8 = DotProduct(A=LinearInterpolate(A=ScalarParameter() {group=Alpha, parameterName=BoltAlpha}, B=@n5, Alpha=@n7), B=@n6.R)
  n9 = DotProduct(A=@n5, B=@n6.R)
  Multiply(
   A=StaticSwitchParameter(True=Saturate(None=@n8), False=Saturate(None=@n9)) {defaultValue=true, parameterName=UseBolt}
   B=ScalarParameter() {defaultValue=1, parameterName=Final_Alpha}
  )
  n10 = LinearInterpolate(
   A=VectorParameter() {parameterName=BoltColor, defaultValue=<rgba 0 0 1 1>}.RGB
   B=VectorParameter() {parameterName=BoltColor_Stop, defaultValue=<rgba 0 0 1 1>}.RGB
   Alpha=ScalarParameter() {parameterName=AnimateBoltColor}
  )
  n11 = CollectionParameter() {collection=/BattleRoyaleGameUI/Shared/MPC_BRGameUI, parameterName=Cyan}
  n12 = LinearInterpolate(
   A=LinearInterpolate(
    A=MaterialFunctionCall!(BufferSource) {materialFunction=/BattleRoyaleGameUI/Shared/MF_UI_AdaptiveColors}.Dark
    B=VectorParameter() {parameterName=ProgressBackingColorStop, defaultValue=<rgba 0 1 0 1>}.RGB
    Alpha=ScalarParameter() {parameterName=AnimateProgressBackingColor}
   )
   B=LinearInterpolate(
    A=VectorParameter() {parameterName=ProgressFillColor, defaultValue=<rgba 1 0 0 1>}.RGB
    B=ComponentMask(None=@n11) {b=true, g=true, r=true}
    Alpha=ScalarParameter() {parameterName=UseMPCSystemColor}
   )
   Alpha=@n4
  )
  StaticSwitchParameter(True=StaticSwitchParameter(True=LinearInterpolate(A=@n10, B=@n12, Alpha=@n7), False=@n12) {parameterName=HasUniqueBoltColor}, False=@n12) {defaultValue=true, parameterName=UseBolt}
 ]
}
